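# Physics asset for the armband on the male medium Knight Remix player body cosmetic, defining collision and joints for its dyn_armband bones.
# Capsule bodies cover the knot and two equally sized strands, hi and low.
# Each strand is constrained directly to the knot with moderate swing limits and a tight twist limit, forming two branches from the knot with no constraint between the strands.
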PhysicsAsset /BRCosmetics/Characters/Player/Male/Medium/Bodies/M_MED_Knight_Remix/Meshes/M_MED_Knight_Remix_Armband_Physics
bodies {
 dyn_armband_hi [{shapeType=Capsule, length=6, radius=2, center=<3 0 0>, rotation=<rot 90 0 0>}]
 dyn_armband_knot [{shapeType=Capsule, length=2.2482, radius=2.2439, center=<1.1374 -0.1907 -0.066>, rotation=<rot 41.9282 -13.7965 -90>}]
 dyn_armband_low [{shapeType=Capsule, length=6, radius=2, center=<3 0 0>, rotation=<rot 90 0 0>}]
}
constraints [
 {bone1Name=dyn_armband_low, bone2Name=dyn_armband_knot, swing1LimitDegrees=60, swing2LimitDegrees=60, twistLimitDegrees=5}
 {bone1Name=dyn_armband_hi, bone2Name=dyn_armband_knot, swing1LimitDegrees=60, swing2LimitDegrees=60, twistLimitDegrees=5}
]
