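# Physics asset for the Knight Remix male medium body's skirt, covering both left and right sides despite its "Right" name.
# Capsules cover pelvis, thighs, calves and feet, with free angular motion between leg segments.
# Four four-box dyn_jacket chains per side hang from each thigh, with narrow twist limits and tighter mid-front swing.
# Angular-locked cross-constraints tie each side's front and front-side chains together at the third bone.
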PhysicsAsset /BRCosmetics/Characters/Player/Male/Medium/Bodies/M_MED_Knight_Remix/Meshes/M_MED_Knight_Remix_Skirt_Right_Physics
bodies {
 calf_l [
  {shapeType=Capsule, length=34.25, radius=9.25, center=<-19.3279 -0.0009 0>, rotation=<rot 90 0 -179.9999>}
  {shapeType=Capsule, length=41, radius=4.75, center=<-19.8776 -6.4469 -7.4685>, rotation=<rot -90 165.9637 -140.8828>}
 ]
 calf_r [{shapeType=Capsule, length=35.25, radius=8.25, center=<19.3279 0.0009 0>, rotation=<rot 90 0 0>}]
 dyn_jacket_back_01_l [{shapeType=Box, extentX=17.5, extentY=2.5, extentZ=17, center=<8 0 0>, rotation=<rot 90 0 0>}]
 dyn_jacket_back_01_r [{shapeType=Box, extentX=17.5, extentY=2.5, extentZ=17, center=<8 0 0>, rotation=<rot 90 0 0>}]
 dyn_jacket_back_02_l [{shapeType=Box, extentX=17.5, extentY=2.5, extentZ=17, center=<8 0 0>, rotation=<rot 90 0 0>}]
 dyn_jacket_back_02_r [{shapeType=Box, extentX=17.5, extentY=2.5, extentZ=17, center=<8 0 0>, rotation=<rot 90 0 0>}]
 dyn_jacket_back_03_l [{shapeType=Box, extentX=17.5, extentY=2.5, extentZ=17, center=<8 0 0>, rotation=<rot 90 0 0>}]
 dyn_jacket_back_03_r [{shapeType=Box, extentX=17.5, extentY=2.5, extentZ=17, center=<8 0 0>, rotation=<rot 90 0 0>}]
 dyn_jacket_back_04_l [{shapeType=Box, extentX=17.5, extentY=2.5, extentZ=17, center=<8 0 0>, rotation=<rot 90 0 0>}]
 dyn_jacket_back_04_r [{shapeType=Box, extentX=17.5, extentY=2.5, extentZ=17, center=<8 0 0>, rotation=<rot 90 0 0>}]
 dyn_jacket_back_side_01_l [{shapeType=Box, extentX=17.5, extentY=2.5, extentZ=17, center=<8 0 0>, rotation=<rot 90 0 0>}]
 dyn_jacket_back_side_01_r [{shapeType=Box, extentX=17.5, extentY=2.5, extentZ=17, center=<8 0 0>, rotation=<rot 90 0 0>}]
 dyn_jacket_back_side_02_l [{shapeType=Box, extentX=17.5, extentY=2.5, extentZ=17, center=<8 0 0>, rotation=<rot 90 0 0>}]
 dyn_jacket_back_side_02_r [{shapeType=Box, extentX=17.5, extentY=2.5, extentZ=17, center=<8 0 0>, rotation=<rot 90 0 0>}]
 dyn_jacket_back_side_03_l [{shapeType=Box, extentX=21.5, extentY=4.5, extentZ=17, center=<8 0 0>, rotation=<rot 90 0 0>}]
 dyn_jacket_back_side_03_r [{shapeType=Box, extentX=17.5, extentY=2.5, extentZ=17, center=<8 0 0>, rotation=<rot 90 0 0>}]
 dyn_jacket_back_side_04_l [{shapeType=Box, extentX=21.5, extentY=4.5, extentZ=17, center=<8 0 0>, rotation=<rot 90 0 0>}]
 dyn_jacket_back_side_04_r [{shapeType=Box, extentX=17.5, extentY=2.5, extentZ=17, center=<8 0 0>, rotation=<rot 90 0 0>}]
 dyn_jacket_front_01_l [{shapeType=Box, extentX=17.5, extentY=2.5, extentZ=17, center=<8 0 0>, rotation=<rot 90 0 0>}]
 dyn_jacket_front_01_r [{shapeType=Box, extentX=17.5, extentY=2.5, extentZ=17, center=<8 0 0>, rotation=<rot 90 0 0>}]
 dyn_jacket_front_02_l [{shapeType=Box, extentX=17.5, extentY=2.5, extentZ=17, center=<8 0 0>, rotation=<rot 90 0 0>}]
 dyn_jacket_front_02_r [{shapeType=Box, extentX=17.5, extentY=2.5, extentZ=17, center=<8 0 0>, rotation=<rot 90 0 0>}]
 dyn_jacket_front_03_l [{shapeType=Box, extentX=21.5, extentY=4.5, extentZ=17, center=<8 0 0>, rotation=<rot 90 0 0>}]
 dyn_jacket_front_03_r [{shapeType=Box, extentX=17.5, extentY=2.5, extentZ=17, center=<8 0 0>, rotation=<rot 90 0 0>}]
 dyn_jacket_front_04_l [{shapeType=Box, extentX=21.5, extentY=4.5, extentZ=17, center=<8 0 0>, rotation=<rot 90 0 0>}]
 dyn_jacket_front_04_r [{shapeType=Box, extentX=17.5, extentY=2.5, extentZ=17, center=<8 0 0>, rotation=<rot 90 0 0>}]
 dyn_jacket_front_side_01_l [{shapeType=Box, extentX=17.5, extentY=2.5, extentZ=17, center=<8 0 0>, rotation=<rot 90 0 0>}]
 dyn_jacket_front_side_01_r [{shapeType=Box, extentX=17.5, extentY=2.5, extentZ=17, center=<8 0 0>, rotation=<rot 90 0 0>}]
 dyn_jacket_front_side_02_l [{shapeType=Box, extentX=17.5, extentY=2.5, extentZ=17, center=<8 0 0>, rotation=<rot 90 0 0>}]
 dyn_jacket_front_side_02_r [{shapeType=Box, extentX=17.5, extentY=2.5, extentZ=17, center=<8 0 0>, rotation=<rot 90 0 0>}]
 dyn_jacket_front_side_03_l [{shapeType=Box, extentX=21.5, extentY=4.5, extentZ=17, center=<8 0 0>, rotation=<rot 90 0 0>}]
 dyn_jacket_front_side_03_r [{shapeType=Box, extentX=17.5, extentY=2.5, extentZ=17, center=<8 0 0>, rotation=<rot 90 0 0>}]
 dyn_jacket_front_side_04_l [{shapeType=Box, extentX=21.5, extentY=4.5, extentZ=17, center=<8 0 0>, rotation=<rot 90 0 0>}]
 dyn_jacket_front_side_04_r [{shapeType=Box, extentX=17.5, extentY=2.5, extentZ=17, center=<8 0 0>, rotation=<rot 90 0 0>}]
 foot_l [{shapeType=Capsule, length=16, radius=8.25, center=<-3.926 5.7336 0>, rotation=<rot 90 0 90>}]
 foot_r [{shapeType=Capsule, length=16, radius=8.25, center=<3.9317 -5.6438 -0.9888>, rotation=<rot 90 0 90>}]
 pelvis [{shapeType=Capsule, length=11.8397, radius=11.0796, center=<-10.0032 -2.3254 0.1947>}]
 thigh_l [{shapeType=Capsule, length=22, radius=10, center=<-25 0 0>, rotation=<rot 90 0 -179.9999>}]
 thigh_r [{shapeType=Capsule, length=22, radius=10, center=<25 0 0>, rotation=<rot 90 0 0.0001>}]
}
constraints [
 {
  bone1Name=calf_l
  bone2Name=thigh_l
  swing1LimitDegrees=45
  swing1Motion=Free
  swing2LimitDegrees=45
  swing2Motion=Free
  twistLimitDegrees=45
  twistMotion=Free
 }
 {
  bone1Name=foot_r
  bone2Name=calf_r
  swing1LimitDegrees=45
  swing1Motion=Free
  swing2LimitDegrees=45
  swing2Motion=Free
  twistLimitDegrees=45
  twistMotion=Free
 }
 {
  bone1Name=foot_l
  bone2Name=calf_l
  swing1LimitDegrees=45
  swing1Motion=Free
  swing2LimitDegrees=45
  swing2Motion=Free
  twistLimitDegrees=45
  twistMotion=Free
 }
 {
  bone1Name=calf_r
  bone2Name=thigh_r
  swing1LimitDegrees=45
  swing1Motion=Free
  swing2LimitDegrees=45
  swing2Motion=Free
  twistLimitDegrees=45
  twistMotion=Free
 }
 {
  bone1Name=thigh_l
  bone2Name=pelvis
  swing1LimitDegrees=45
  swing1Motion=Free
  swing2LimitDegrees=45
  swing2Motion=Free
  twistLimitDegrees=45
  twistMotion=Free
 }
 {
  bone1Name=thigh_r
  bone2Name=pelvis
  swing1LimitDegrees=45
  swing1Motion=Free
  swing2LimitDegrees=45
  swing2Motion=Free
  twistLimitDegrees=45
  twistMotion=Free
 }
 {bone1Name=dyn_jacket_back_01_r, bone2Name=thigh_r, swing1LimitDegrees=45, swing2LimitDegrees=45, twistLimitDegrees=5}
 {bone1Name=dyn_jacket_back_02_r, bone2Name=dyn_jacket_back_01_r, swing1LimitDegrees=45, swing2LimitDegrees=45, twistLimitDegrees=5}
 {bone1Name=dyn_jacket_back_03_r, bone2Name=dyn_jacket_back_02_r, swing1LimitDegrees=45, swing2LimitDegrees=45, twistLimitDegrees=5}
 {bone1Name=dyn_jacket_back_04_r, bone2Name=dyn_jacket_back_03_r, swing1LimitDegrees=45, swing2LimitDegrees=45, twistLimitDegrees=5}
 {bone1Name=dyn_jacket_back_side_01_r, bone2Name=thigh_r, swing1LimitDegrees=45, swing2LimitDegrees=45, twistLimitDegrees=5}
 {bone1Name=dyn_jacket_back_side_02_r, bone2Name=dyn_jacket_back_side_01_r, swing1LimitDegrees=45, swing2LimitDegrees=45, twistLimitDegrees=5}
 {bone1Name=dyn_jacket_back_side_03_r, bone2Name=dyn_jacket_back_side_02_r, swing1LimitDegrees=45, swing2LimitDegrees=45, twistLimitDegrees=5}
 {bone1Name=dyn_jacket_back_side_04_r, bone2Name=dyn_jacket_back_side_03_r, swing1LimitDegrees=45, swing2LimitDegrees=45, twistLimitDegrees=5}
 {bone1Name=dyn_jacket_front_side_01_r, bone2Name=thigh_r, swing1LimitDegrees=45, swing2LimitDegrees=45, twistLimitDegrees=5}
 {
  bone1Name=dyn_jacket_front_side_02_r
  bone2Name=dyn_jacket_front_side_01_r
  swing1LimitDegrees=45
  swing2LimitDegrees=45
  twistLimitDegrees=5
 }
 {
  bone1Name=dyn_jacket_front_side_03_r
  bone2Name=dyn_jacket_front_side_02_r
  swing1LimitDegrees=45
  swing2LimitDegrees=45
  twistLimitDegrees=5
 }
 {
  bone1Name=dyn_jacket_front_side_04_r
  bone2Name=dyn_jacket_front_side_03_r
  swing1LimitDegrees=45
  swing2LimitDegrees=45
  twistLimitDegrees=5
 }
 {bone1Name=dyn_jacket_front_01_r, bone2Name=thigh_r, swing1LimitDegrees=45, swing2LimitDegrees=45, twistLimitDegrees=5}
 {bone1Name=dyn_jacket_front_02_r, bone2Name=dyn_jacket_front_01_r, swing1LimitDegrees=20, swing2LimitDegrees=20, twistLimitDegrees=1}
 {bone1Name=dyn_jacket_front_03_r, bone2Name=dyn_jacket_front_02_r, swing1LimitDegrees=20, swing2LimitDegrees=20, twistLimitDegrees=1}
 {bone1Name=dyn_jacket_front_04_r, bone2Name=dyn_jacket_front_03_r, swing1LimitDegrees=45, swing2LimitDegrees=45, twistLimitDegrees=1}
 {bone1Name=dyn_jacket_back_04_l, bone2Name=dyn_jacket_back_03_l, swing1LimitDegrees=45, swing2LimitDegrees=45, twistLimitDegrees=5}
 {bone1Name=dyn_jacket_front_01_l, bone2Name=thigh_l, swing1LimitDegrees=45, swing2LimitDegrees=45, twistLimitDegrees=5}
 {bone1Name=dyn_jacket_front_02_l, bone2Name=dyn_jacket_front_01_l, swing1LimitDegrees=20, swing2LimitDegrees=20, twistLimitDegrees=5}
 {bone1Name=dyn_jacket_front_03_l, bone2Name=dyn_jacket_front_02_l, swing1LimitDegrees=20, swing2LimitDegrees=20, twistLimitDegrees=5}
 {bone1Name=dyn_jacket_front_04_l, bone2Name=dyn_jacket_front_03_l, swing1LimitDegrees=45, swing2LimitDegrees=45, twistLimitDegrees=5}
 {bone1Name=dyn_jacket_front_side_01_l, bone2Name=thigh_l, swing1LimitDegrees=45, swing2LimitDegrees=45, twistLimitDegrees=5}
 {
  bone1Name=dyn_jacket_front_side_02_l
  bone2Name=dyn_jacket_front_side_01_l
  swing1LimitDegrees=45
  swing2LimitDegrees=45
  twistLimitDegrees=5
 }
 {
  bone1Name=dyn_jacket_front_side_03_l
  bone2Name=dyn_jacket_front_side_02_l
  swing1LimitDegrees=45
  swing2LimitDegrees=45
  twistLimitDegrees=5
 }
 {
  bone1Name=dyn_jacket_front_side_04_l
  bone2Name=dyn_jacket_front_side_03_l
  swing1LimitDegrees=45
  swing2LimitDegrees=45
  twistLimitDegrees=5
 }
 {bone1Name=dyn_jacket_back_side_01_l, bone2Name=thigh_l, swing1LimitDegrees=45, swing2LimitDegrees=45, twistLimitDegrees=5}
 {bone1Name=dyn_jacket_back_side_02_l, bone2Name=dyn_jacket_back_side_01_l, swing1LimitDegrees=45, swing2LimitDegrees=45, twistLimitDegrees=5}
 {bone1Name=dyn_jacket_back_side_03_l, bone2Name=dyn_jacket_back_side_02_l, swing1LimitDegrees=45, swing2LimitDegrees=45, twistLimitDegrees=5}
 {bone1Name=dyn_jacket_back_side_04_l, bone2Name=dyn_jacket_back_side_03_l, swing1LimitDegrees=45, swing2LimitDegrees=45, twistLimitDegrees=5}
 {bone1Name=dyn_jacket_back_01_l, bone2Name=thigh_l, swing1LimitDegrees=45, swing2LimitDegrees=45, twistLimitDegrees=5}
 {bone1Name=dyn_jacket_back_02_l, bone2Name=dyn_jacket_back_01_l, swing1LimitDegrees=45, swing2LimitDegrees=45, twistLimitDegrees=5}
 {bone1Name=dyn_jacket_back_03_l, bone2Name=dyn_jacket_back_02_l, swing1LimitDegrees=45, swing2LimitDegrees=45, twistLimitDegrees=5}
 {
  bone1Name=dyn_jacket_front_side_03_r
  bone2Name=dyn_jacket_front_03_r
  swing1LimitDegrees=45
  swing1Motion=Locked
  swing2LimitDegrees=45
  swing2Motion=Locked
  twistLimitDegrees=45
  twistMotion=Locked
 }
 {
  bone1Name=dyn_jacket_front_side_03_l
  bone2Name=dyn_jacket_front_03_l
  swing1LimitDegrees=45
  swing1Motion=Locked
  swing2LimitDegrees=45
  swing2Motion=Locked
  twistLimitDegrees=45
  twistMotion=Locked
 }
]
